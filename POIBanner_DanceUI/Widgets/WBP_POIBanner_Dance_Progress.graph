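# Progress widget for the dance POI banner capture point, extending WBP_POIWin_ProgressWidget with an emote prompt and an acceleration glow.
# Setup shows the radial bar, creates its dynamic material, and binds to the capture point's dance accelerators, unbinding on Destruct.
# Claim progress updates team icon and color, while friendly emote acceleration shows the glow and animated segment.
# Local emoting and contested state control the "EMOTE!" prompt; emote events bind only for static widgets.
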WidgetBlueprint /POIBanner_DanceUI/Widgets/WBP_POIBanner_Dance_Progress
graphs [
 EventGraph {nodes [
   Event Construct()
   > Parent: Construct()
   Event Destruct()
   > Parent: Destruct()
   > UnbindEvents()
  ]}
 On Progress Contested {nodes [
   n1 = On Progress Contested!() -> (then, IsContested:bool) {is=title K2Node_FunctionEntry}
   > Parent: On Progress Contested!(execute, IsContested:bool=@n1) -> (then) {is=title K2Node_CallParentFunction}
   n2 = Get IsLocalPlayerEmoting() -> (IsLocalPlayerEmoting:bool)
   > UpdateEmotePromptVisibility!(
    execute
    self
    DisplayPrompt:bool=AND Boolean!(A=Get IsStaticWidget() -> (IsStaticWidget:bool), B=NOT Boolean(A=@n1), C:bool=NOT Boolean(A=@n2)) -> (ReturnValue)
   ) -> (then)
   > UpdateEmoteAccelerationVisuals!(execute, self, HasEmoteAcceleration:bool=false) -> (then)
  ]}
 OnAccelerationActiveUpdated {nodes [
   n1 = OnAccelerationActiveUpdated!() -> (then, AccelerationActive:bool=~) {is=title K2Node_FunctionEntry}
   > UpdateEmoteAccelerationVisuals(HasEmoteAcceleration=@n1)
  ]}
 OnNewPawnEmotingUpdated {nodes [
   n1 = OnNewPawnEmotingUpdated!() -> (then, FortPlayerPawn:Fort Player Pawn Object Reference, IsEmoting:bool=~) {is=title K2Node_FunctionEntry}
   > Set IsLocalPlayerEmoting(IsLocalPlayerEmoting:bool=@n1.IsEmoting) -> (Output_Get:bool)
   > n2 = Cast To FortPlayerController(Object=GetOwningPlayer())
   > n3 = Cast To PlayerPawn_Athena(Object=GetViewTarget!(self:Controller Object Reference=@n2) -> (ReturnValue:Actor Object Reference))
   > Branch(Condition="Equal (Object)"(A=@n3, B=@n1))
   n4 = Get IsLocalPlayerEmoting()
   n5 = Get IsContested() -> (IsContested:bool)
   > UpdateEmotePromptVisibility(DisplayPrompt=AND Boolean!(A=NOT Boolean(A=@n4), B=NOT Boolean(A=@n5)) -> (ReturnValue))
  ]}
 Setup Progress Materials {nodes [
   Setup Progress Materials() {is=title K2Node_FunctionEntry}
   n1 = Get Image_Radial_ProgressBar() -> (Image_Radial_ProgressBar:Image Object Reference)
   > SetBrushFromMaterial(self=@n1, Material=/POIBanner_DanceUI/Art/Materials/MI_UI_POIBanner_Dance_RadialFill)
   > n2 = GetDynamicMaterial(self=@n1)
   > n3 = Set RadialDynamicMaterial(RadialDynamicMaterial:Material Instance Dynamic Object Reference=@n2) -> (Output_Get:Material Instance Dynamic Object Reference)
   > SetVectorParameterValue(self=@n3, ParameterName=BGColor, Value=(R=0.0953,G=0.0953,B=0.0953,A=1))
   > SetScalarParameterValue(self=@n3, ParameterName=DisplaySegmentAnim, Value=0)
   > Update Progress Value!(execute, self:WBP POIWin Progress Widget Object Reference, Parameter Name:Name=Progress, Value:double=0) -> (then)
  ]}
 SetupProgressBarToShow {nodes [
   n1 = SetupProgressBarToShow!() -> (then, CapturePointActor:Actor Object Reference) {is=title K2Node_FunctionEntry}
   > SetActiveWidget(
    self=Get Switcher_ProgressBar_Shape() -> (Switcher_ProgressBar_Shape:Common Visibility Switcher Object Reference)
    Widget=Get Overlay_Radial() -> (Overlay_Radial:Overlay Object Reference)
   )
   > n2 = Cast To BGA_POIBanner_CapturePoint_Dance!(execute, Object=@n1) -> (then, CastFailed, AsBGA POIBanner Capture Point Dance:BGA POIBanner Capture Point Dance Object Reference) {is=title K2Node_DynamicCast}
   > n3 = For Each Loop(Array:Array of BGA POIWin Capture Point Accelerator Object References=Get Accelerators!(self:BGA POIBanner Capture Point Object Reference=@n2) -> (Accelerators:Array of BGA POIWin Capture Point Accelerator Object References)) -> (Array Element:BGA POIWin Capture Point Accelerator Object Reference)
   n4 = Get IsContested()
   UpdateEmotePromptVisibility(execute=@n3.Completed, DisplayPrompt=AND Boolean(A=Get IsStaticWidget(), B=NOT Boolean(A=@n4)))
   Is Valid(exec=@n3.LoopBody, InputObject=@n3)
   > n5 = Cast To BGA_POIWin_CapturePoint_Accelerator_Dance!(execute, Object=@n3) -> (then, CastFailed, AsBGA POIWin Capture Point Accelerator Dance:BGA POIWin Capture Point Accelerator Dance Object Reference) {is=title K2Node_DynamicCast}
   n6 = Get BoundAccelerators() -> (BoundAccelerators:Array of BGA POIWin Capture Point Accelerator Dance Object References)
   > n7 = Branch(Condition=Is Valid Index!(
     TargetArray:Array of BGA POIWin Capture Point Accelerator Dance Object References=@n6
     IndexToTest:int=Find Item!(
      TargetArray:Array of BGA POIWin Capture Point Accelerator Dance Object References=@n6
      ItemToFind:"BGA POIWin Capture Point Accelerator Dance Object Reference (by ref)"=@n5
     ) -> (ReturnValue)
    ) -> (ReturnValue:bool))
   Add(
    execute=@n7.else
    TargetArray:Array of BGA POIWin Capture Point Accelerator Dance Object References=Get BoundAccelerators()
    NewItem:"BGA POIWin Capture Point Accelerator Dance Object Reference (by ref)"=@n5
   ) {is=title K2Node_CallArrayFunction}
   > Bind Event to On Rep Acceleration Updated(self:BGA POIWin Capture Point Accelerator Object Reference=@n5, Delegate=Create Event()) {is=title K2Node_AddDelegate}
   > Branch(Condition=Get IsStaticWidget())
   > Bind Event to On Pawn Emoting Updated(self:BGA POIWin Capture Point Accelerator Dance Object Reference=@n5, Delegate=Create Event()) {is=title K2Node_AddDelegate}
  ]}
 UnbindEvents {nodes [
   UnbindEvents() {is=title K2Node_FunctionEntry}
   > n1 = For Each Loop(Array=Get BoundAccelerators())
   Clear(execute=@n1.Completed, TargetArray:Array of BGA POIWin Capture Point Accelerator Dance Object References=Get BoundAccelerators())
   Is Valid(exec=@n1.LoopBody, InputObject=@n1)
   > n2 = Cast To BGA_POIWin_CapturePoint_Accelerator_Dance(Object=@n1)
   > Unbind Event from On Rep Acceleration Updated(self:BGA POIWin Capture Point Accelerator Object Reference=@n2, Delegate=Create Event()) {is=title K2Node_RemoveDelegate}
   > Branch(Condition=Get IsStaticWidget())
   > Unbind Event from On Pawn Emoting Updated(self:BGA POIWin Capture Point Accelerator Dance Object Reference=@n2, Delegate=Create Event()) {is=title K2Node_RemoveDelegate}
  ]}
 Update Claim Progress {nodes [
   n1 = Update Claim Progress!() -> (then, IsFriendlyTeam:bool, CurrentProgress:double) {is=title K2Node_FunctionEntry}
   > Set Is Friendly Team(Is Friendly Team:bool=@n1) -> (Output_Get:bool)
   > Update Progress Value(Parameter Name=Progress, Value=@n1.CurrentProgress)
   > SetBrushFromTexture!(
    execute
    self:Image Object Reference=Get Image_StatusIcon() -> (Image_StatusIcon:Image Object Reference)
    Texture:Texture 2D Object Reference=Select(
     Option 0:Texture 2D Object Reference=Get EnemyProgress() -> (EnemyProgress:Texture 2D Object Reference)
     Option 1:Texture 2D Object Reference=Get FriendlyProgress() -> (FriendlyProgress:Texture 2D Object Reference)
     Index=Get IsFriendlyTeam() -> (IsFriendlyTeam:bool)
    ) -> (ReturnValue:Texture 2D Object Reference)
    bMatchSize:bool=false
   ) -> (then)
   n2 = Get IsLocalPlayerEmoting()
   n3 = Get AccelerationColor() -> (AccelerationColor:Linear Color Structure)
   n4 = Get FriendlyColor() -> (FriendlyColor:Linear Color Structure)
   > Set Progress Color!(
    execute
    self:WBP POIWin Progress Widget Object Reference
    Value:Linear Color Structure=SelectColor(A=SelectColor(A=@n3, B=@n4, bPickA=@n2), B=Get EnemyColor() -> (EnemyColor:Linear Color Structure), bPickA=Get IsFriendlyTeam())
   ) -> (then) {is=title K2Node_CallFunction}
  ]}
 UpdateEmoteAccelerationVisuals {nodes [
   n1 = UpdateEmoteAccelerationVisuals!() -> (then, HasEmoteAcceleration:bool=~) {is=title K2Node_FunctionEntry}
   > n2 = Branch(Condition=AND Boolean(A=@n1, B=Get Is Friendly Team() -> (Is Friendly Team:bool)))
   > SetVisibility(self=Get Image_BGGlow() -> (Image_BGGlow:Image Object Reference), InVisibility=HitTestInvisible)
   n3 = Get RadialDynamicMaterial() -> (RadialDynamicMaterial:Material Instance Dynamic Object Reference)
   > SetScalarParameterValue(self=@n3, ParameterName=DisplaySegmentAnim, Value=1)
   > SetVectorParameterValue(self=@n3, ParameterName=Shape Color, Value=(R=0.7991,G=0.3613,B=0.0382,A=1))
   SetVisibility(execute=@n2.else, self=Get Image_BGGlow(), InVisibility=Hidden)
   > SetScalarParameterValue(self=Get RadialDynamicMaterial(), ParameterName=DisplaySegmentAnim, Value=0)
   n4 = Get FriendlyColor()
   n5 = Get EnemyColor()
   n6 = Get Is Friendly Team()
   > Set Progress Color(Value=Select(
     Option 0:Linear Color Structure=Select(Option 0=@n5, Option 1=@n4, Index=@n6)
     Option 1:Linear Color Structure=Get ContestedColor() -> (ContestedColor:Linear Color Structure)
     Index=Get IsContested()
    ) -> (ReturnValue:Linear Color Structure))
  ]}
 UpdateEmotePromptVisibility {nodes [
   n1 = UpdateEmotePromptVisibility!() -> (then, DisplayPrompt:bool=~) {is=title K2Node_FunctionEntry}
   > SetVisibility(
    self=Get Text_EmotePrompt() -> (Text_EmotePrompt:Common Text Object Reference)
    InVisibility=Select(Option 0:ESlateVisibility Enum=Collapsed, Option 1:ESlateVisibility Enum=HitTestInvisible, Index=@n1) -> (ReturnValue:ESlateVisibility Enum)
   )
  ]}
]
variables [
 AccelerationColor {type=LinearColor, default=<rgba 0.7991 0.3613 0.0382 1>}
 BoundAccelerators {type="array<BGA_POIWin_CapturePoint_Accelerator_Dance_C>", default []}
 IsLocalPlayerEmoting {type=boolean}
]
widgets {
 info {
  inheritedWidgetCount=0
  namedSlotCount=1
  parentClass=/POIBanner/UI/WBP_POIWin_ProgressWidget.WBP_POIWin_ProgressWidget_C
  rootWidgetClass=null
  widgetCount=3
 }
 widgets [
  Overlay_0 {widgetClassPath=&Overlay, widget=Overlay_0}
  Image_BGGlow {
   widgetClassPath=&Image
   bIsVariable=true
   parent=Overlay_0
   widget=Image_BGGlow
   properties {
    visibility=Collapsed
    brush {
     bIsDynamicallyLoaded=false
     drawAs=Image
     imageType=FullColor
     mirroring=NoMirror
     resourceName=None
     resourceObject=/Game/Athena/HUD/LTM/Wax/Materials/M_UI_TextGlow
     tiling=NoTile
     imageSize=<32 32>
     margin=<ltrb 0 0 0 0>
     outlineSettings {
      bUseBrushTransparency=false
      roundingType=HalfHeightRadius
      width=0
      color {colorUseRule=UseColor_Specified, specifiedColor=<rgba 0 0 0 0>}
      cornerRadii=<0 0 0 0>
     }
     tintColor {colorUseRule=UseColor_Specified, specifiedColor=<rgba 0.9216 0.5089 0.0782 1>}
     uVRegion {bIsValid=false, max=<0 0>, min=<0 0>}
    }
   }
   slot {class=OverlaySlot, properties {horizontalAlignment=HAlign_Fill, verticalAlignment=VAlign_Fill, padding=<ltrb -50 -50 -50 -50>}}
  }
  Text_EmotePrompt {
   widgetClassPath=&CommonTextBlock
   bIsVariable=true
   parent=Overlay_0
   widget=Text_EmotePrompt
   properties {
    justification=Center
    scrollStyle=&TextScrollStyle-Base_C
    text="EMOTE!"
    textTransformPolicy=ToUpper
    visibility=Collapsed
    colorAndOpacity {colorUseRule=UseColor_Specified, specifiedColor=<rgba 0.7991 0.3613 0.0382 1>}
    font {
     bForceMonospaced=false
     bMaterialIsStencil=false
     fontMaterial=null
     fontObject=&BurbankSmall
     letterSpacing=0
     monospacedWidth=1
     size=22.5
     skewAmount=0
     typefaceFontName=Black
     outlineSettings {
      bApplyOutlineToDropShadows=false
      bMiteredCorners=false
      bSeparateFillAlpha=false
      outlineMaterial=null
      outlineSize=1
      outlineColor=<rgba 0 0 0 0.5>
     }
    }
    shadowOffset=<0 0>
   }
   slot {class=OverlaySlot, properties {horizontalAlignment=HAlign_Center, verticalAlignment=VAlign_Bottom, padding=<ltrb -10 0 -10 -50>}}
  }
 ]
}
